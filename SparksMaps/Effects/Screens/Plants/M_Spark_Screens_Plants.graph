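# Unlit material for a Sparks screen showing animated plant shapes that react to music.
# Three scaled, rotating samples of one plant texture form tinted leaf layers and a mask, alongside a dark flower on a stage-or-fixed color gradient.
# Audio measure time drives curve-atlas beat pulses, with panning sparkle and noise-wobble signals also built.
# Named reroutes link the parts, and the composite feeds the shared screen-details function.
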Material /SparksMaps/Effects/Screens/Plants/M_Spark_Screens_Plants
graph {
 properties {bAutomaticallySetUsageInEditor=false, shadingModel=MSM_Unlit, usedShadingModels=Unlit}
 nodes [
  Constant2Vector() {g=1.3, r=1.3}
  Constant2Vector() {g=0.6, r=0.6}
  Constant2Vector() {g=2.7, r=2.7}
  Constant() {r=0.5}
  Constant() {r=0.5}
  Constant() {r=0.35}
  Constant() {r=0.35}
  Constant() {r=0.55}
  Constant() {r=0.55}
  n1 = NamedRerouteUsage()
  n2 = TextureCoordinate()
  n3 = ScalarParameter() {defaultValue=1, parameterName=GreenScaleX}
  n4 = ScalarParameter() {defaultValue=1, parameterName=GreenScaleY}
  n5 = Multiply(
   A=MaterialFunctionCall(UVs=Add(A=@n1, B=@n2), Texture Scale=AppendVector(A=@n3, B=@n4)) {materialFunction=&ScaleUVsByCenter}.UVs
   B=AppendVector(A=ScalarParameter() {defaultValue=0.5, parameterName=GreenSlideX}, B=ScalarParameter() {defaultValue=0.5, parameterName=GreenSlideY})
  )
  n6 = TextureSample(UVs=MaterialFunctionCall(UVs=Rotator(Coordinate=@n5) {speed=-0.02}, "Rotation Angle (0-1)"=ScalarParameter() {defaultValue=0.4, parameterName=LeafRotateBig}) {materialFunction=&CustomRotator}.Rotated Values) {samplerSource=SSM_Clamp_WorldGroupSettings, texture=&/T_Sparks_Screen_Plants}
  n7 = NamedRerouteUsage()
  n8 = TextureCoordinate()
  n9 = ScalarParameter() {defaultValue=1.3, parameterName=PurpScaleX}
  n10 = ScalarParameter() {defaultValue=1.3, parameterName=PurpScaleY}
  n11 = Multiply(
   A=MaterialFunctionCall(UVs=Add(A=@n7, B=@n8), Texture Scale=AppendVector(A=@n9, B=@n10)) {materialFunction=&ScaleUVsByCenter}.UVs
   B=AppendVector(A=ScalarParameter() {defaultValue=0.55, parameterName=PurpSlideX}, B=ScalarParameter() {defaultValue=0.55, parameterName=PurpSlideY})
  )
  n12 = TextureSample(UVs=MaterialFunctionCall(UVs=Rotator(Coordinate=@n11) {speed=-0.02}, "Rotation Angle (0-1)"=ScalarParameter() {defaultValue=0.2, parameterName=LeafRotateMid}) {materialFunction=&CustomRotator}.Rotated Values) {samplerSource=SSM_Clamp_WorldGroupSettings, texture=&T_Sparks_Screen_Plants}
  n13 = Add(A=Multiply(A=NamedRerouteUsage(), B=@n6.R), B=Multiply(A=NamedRerouteUsage(), B=@n12.R))
  n14 = MaterialFunctionCall(
   UVs=Add(A=NamedRerouteUsage(), B=TextureCoordinate())
   Texture Scale=AppendVector(A=ScalarParameter() {defaultValue=0.6, parameterName=PinkScale}, B=ScalarParameter() {defaultValue=0.6, parameterName=PinkScaleY})
  ) {materialFunction=&ScaleUVsByCenter}
  n15 = AppendVector(A=ScalarParameter() {defaultValue=0.35, parameterName=PinkSlideX}, B=ScalarParameter() {defaultValue=0.35, parameterName=PinkSlideY})
  n16 = TextureSample(UVs=Rotator(Coordinate=Multiply(A=@n14.UVs, B=@n15)) {speed=-0.02}) {samplerSource=SSM_Clamp_WorldGroupSettings, texture=&T_Sparks_Screen_Plants}
  n17 = Multiply(A=NamedRerouteUsage(), B=@n16.R)
  n18 = CurveAtlasRowParameter(CurveTime=Multiply(
    A=CollectionParameter() {collection=/FMAudioReactiveData/AudioAnalysis/MPC_Sparks_Audio_Parameters, parameterName=MeasureLifetime}
    B=ScalarParameter() {defaultValue=4, parameterName=FlowerBeatBright}
   ) {constB=2}) {
   atlas=/SparksMaps/Effects/Screens/Curves/CurveAtlas_Audio_Beat_A
   curve=/SparksMaps/Effects/Screens/Curves/Curve_Audio_Beat_Remap_001
   parameterName=ReactivityCurve
  }
  n19 = Constant()
  n20 = Constant() {r=1}
  n21 = Constant() {r=0.4}
  n22 = ScalarParameter() {defaultValue=0.6, parameterName=FllowerGradientHigh}
  n23 = LinearInterpolate(
   A=NamedRerouteUsage()
   B=Multiply(
    A=Add(A=@n13, B=@n17)
    B=MaterialFunctionCall(Input=@n18, Input Low=@n19, Input High=@n20, Target Low=@n21, Target High=@n22) {materialFunction=&RemapValueRange}.Result
   ) {constB=0.4}
   Alpha=NamedRerouteUsage()
  ) {constB=0}
  n24 = Multiply(A=Multiply(A=NamedRerouteUsage()) {constB=30}, B=@n23)
  n25 = OneMinus(None=Multiply(A=NamedRerouteUsage(), B=ScalarParameter() {defaultValue=1, parameterName=StarBright}) {constB=0.9})
  MaterialFunctionCall!(Emissive=Add(A=Multiply(A=@n24, B=@n25), B=@n23), Scanlines Tile) {materialFunction=/SparksMaps/Effects/Screens/Materials/MF/MF_Sparks_Screen_Details}
  n26 = TextureCoordinate()
  n27 = Constant() {r=2}
  n28 = Add(A=ComponentMask(None=Rotator(Coordinate=@n26, Time=@n27)) {r=true}, B=Multiply(A=Time(), B=Constant() {r=0.05}))
  Multiply(A=@n28, B=Constant() {r=2})
  n29 = Constant2Vector() {g=1, r=1}
  n30 = Constant() {r=0.85}
  n31 = Constant() {r=5}
  NamedRerouteDeclaration(None=LinearInterpolate(
    A=Constant3Vector() {constant=<rgba 0.5208 0.2238 0 1>}
    B=Constant3Vector() {constant=<rgba 0.6198 0.3884 0 1>}
    Alpha=MaterialFunctionCall(CenterPosition=@n29, Radius=@n30, Density=@n31) {materialFunction=&RadialGradientExponential}.RadialGradientExponential
   )) {name=Color1, nodeColor=<rgba 0 0.4375 1 1>}
  n32 = Sine(None=@n28)
  n33 = Multiply(A=TextureCoordinate(), B=AppendVector(A=Constant() {r=0.5}, B=Constant() {r=0.5}))
  n34 = AppendVector(A=Constant() {r=0.15}, B=Constant() {r=0.1})
  n35 = Multiply(
   A=TextureSample(UVs=Panner(Coordinate=@n33, Speed=@n34) {speedX=0.2, speedY=0.15}) {samplerType=SAMPLERTYPE_Grayscale, texture=&LowResBlurredNoise}.R
   B=Constant() {r=2}
  )
  NamedRerouteDeclaration(None=Multiply(A=Add(A=@n32, B=@n35), B=ScalarParameter() {defaultValue=0.01, parameterName=Rays 1 Wobble Amount}) {constB=0.001}) {name=Wobble, nodeColor=<rgba 1 0 0.9141 1>}
  n36 = NamedRerouteUsage()
  n37 = NamedRerouteUsage()
  n38 = TextureCoordinate()
  n39 = ScalarParameter() {defaultValue=2.7, parameterName=BG_ScaleX}
  n40 = ScalarParameter() {defaultValue=2.7, parameterName=BG_ScaleY}
  n41 = Add(
   A=MaterialFunctionCall(UVs=Add(A=@n37, B=@n38), Texture Scale=AppendVector(A=@n39, B=@n40)) {materialFunction=&ScaleUVsByCenter}.UVs
   B=AppendVector(A=ScalarParameter() {defaultValue=0.8, parameterName=BG_Slide_X}, B=ScalarParameter() {defaultValue=0.8, parameterName=BG_Slide_Y})
  )
  n42 = MaterialFunctionCall(
   UVs=Rotator(Coordinate=Subtract(A=@n41)) {speed=-0.02}
   "Rotation Angle (0-1)"=ScalarParameter() {defaultValue=0.005, parameterName=DarkFlowerRotater}
  ) {materialFunction=&CustomRotator}
  n43 = OneMinus(None=Multiply(
    A=TextureSample(UVs=@n42.Rotated Values) {samplerSource=SSM_Clamp_WorldGroupSettings, texture=&T_Sparks_Screen_Plants}.R
    B=ScalarParameter() {defaultValue=2}
   ))
  n44 = MaterialFunctionCall(UVs=TextureCoordinate(), "Rotation Angle (0-1)"=ScalarParameter() {defaultValue=0.375, parameterName=GradientRotation}) {materialFunction=&CustomRotator}
  n45 = Subtract(A=OneMinus(None=ComponentMask(None=@n44.Rotated Values) {r=true}), B=ScalarParameter() {defaultValue=0.2, parameterName=GradientSubtract})
  n46 = ScalarParameter() {defaultValue=10, parameterName=GradientDivide}
  n47 = Multiply(
   A=LinearInterpolate(
    A=MaterialFunctionCall!(CustomUPosition) {materialFunction=/SparksStageShow/Systems/StageManager/FXManager/MaterialFunctions/MF_MojoStageColors}.Primary
    B=VectorParameter() {parameterName=BG_Color, defaultValue=<rgba 0 0.0521 0.0036 1>}.RGB
    Alpha=ScalarParameter() {parameterName=Stage_Or_Fixed_Color}
   )
   B=Saturate(None=Divide(A=@n45, B=@n46))
  ) {constB=0.05}
  n48 = ScalarParameter() {defaultValue=1, parameterName=GradientBright}
  NamedRerouteDeclaration(None=Multiply(A=Multiply(A=@n36, B=@n43), B=Multiply(A=@n47, B=@n48))) {name=DarkFlower, nodeColor=<rgba 0 1 0.6719 1>}
  n49 = Panner(Coordinate=TextureCoordinate() {uTiling=8, vTiling=8}) {speedX=0.02, speedY=0.03}
  n50 = Panner(Coordinate=TextureCoordinate() {uTiling=7, vTiling=7}) {speedX=0.04, speedY=0.04}
  NamedRerouteDeclaration(None=Multiply(
    A=TextureSample(UVs=@n49) {texture=/SparksMaps/Environment/S5/Generic/Textures/T_Sparkle_Spots_B}.R
    B=TextureSample(UVs=@n50) {texture=&T_Sparkle_Spots_B}.R
   )) {name=Sparkles, nodeColor=<rgba 1 0.0234 0 1>}
  n51 = CurveAtlasRowParameter(CurveTime=Multiply(
    A=CollectionParameter() {collection=&MPC_Sparks_Audio_Parameters, parameterName=MeasureLifetime}
    B=ScalarParameter() {defaultValue=1, parameterName=Beat}
   ) {constB=2}) {
   atlas=&CurveAtlas_Audio_Beat_A
   curve=&Curve_Audio_Beat_Remap_001
   parameterName=ReactivityCurve
  }
  NamedRerouteDeclaration(None=MaterialFunctionCall(
    Input=@n51
    Input Low=Constant()
    Input High=Constant() {r=1}
    Target Low=Constant() {r=1}
    Target High=ScalarParameter() {defaultValue=3, parameterName=GradientBeatHigh}
   ) {materialFunction=&RemapValueRange}.Result) {name=Reactivity, nodeColor=<rgba 0 0 0 1>}
  n52 = Constant2Vector() {g=1, r=1}
  n53 = Constant() {r=0.85}
  n54 = Constant() {r=5}
  NamedRerouteDeclaration(None=LinearInterpolate(
    A=Constant3Vector() {constant=<rgba 0.7344 0.4274 0 1>}
    B=Constant3Vector() {constant=<rgba 0.5052 0.2542 0 1>}
    Alpha=MaterialFunctionCall(CenterPosition=@n52, Radius=@n53, Density=@n54) {materialFunction=&RadialGradientExponential}.RadialGradientExponential
   )) {name=Color2, nodeColor=<rgba 1 0 0.4453 1>}
  n55 = Constant2Vector()
  n56 = Constant() {r=0.95}
  n57 = Constant() {r=10}
  NamedRerouteDeclaration(None=LinearInterpolate(
    A=Constant3Vector() {constant=<rgba 0.4635 0.188 0 1>}
    B=Constant3Vector() {constant=<rgba 0.6042 0.2655 0 1>}
    Alpha=MaterialFunctionCall(CenterPosition=@n55, Radius=@n56, Density=@n57) {materialFunction=&RadialGradientExponential}.RadialGradientExponential
   )) {name=Color3, nodeColor=<rgba 1 0 0.0469 1>}
  n58 = Add(A=@n6.R, B=@n12.R)
  NamedRerouteDeclaration(None=MaterialFunctionCall(Input=Add(A=@n58, B=@n16.R), Input Low=Constant(), Input High=Constant() {desc=was .01, r=0.5}) {materialFunction=&RemapValue_0-1}.Result) {name=LeafMask, nodeColor=<rgba 1 0.1406 0 1>}
  ScalarParameter() {defaultValue=1, parameterName=DarkflowerBright}
 ]
}
